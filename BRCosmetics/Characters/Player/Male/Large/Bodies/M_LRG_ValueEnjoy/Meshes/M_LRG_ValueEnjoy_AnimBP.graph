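# Post-process style AnimBlueprint for the large male ValueEnjoy body, copying the attached parent's pose and adding secondary motion to torso weight bones.
# A Control Rig runs first, its alpha switched by OnBeginSliding and OnEndSliding events.
# Chained AnimDynamics nodes add angular springs to weight, front and side bones, then linear springs to front and side.
# Final mirrored, weighted constraints blend weight bones from those physics bones and nearby body bones.
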AnimBlueprint /BRCosmetics/Characters/Player/Male/Large/Bodies/M_LRG_ValueEnjoy/Meshes/M_LRG_ValueEnjoy_AnimBP
target_skeleton=&/M_LRG_ValueEnjoy_Skeleton
graphs [
 AnimGraph {nodes [
   n1 = Control Rig!(Source=Copy Pose From Mesh() {bUseAttachedParent=true}, Alpha, slidingAlpha:double=Get slidingAlpha() -> (slidingAlpha:double)) -> (Pose) {controlRigAssetReference {blueprintRigClass=&/M_LRG_ValueEnjoy_CtrlRig.M_LRG_ValueEnjoy_CtrlRig_C}}
   n2 = Anim Dynamics - Bone: dyn_L_weight_A(ComponentPose=Local To Component(LocalPose=@n1), Alpha=1, ExternalForce="0,0,0") {
    angularDampingOverride=0.999
    angularSpringConstant=2000
    bAngularSpring=true
    bOverrideAngularDamping=true
    bUseGravityOverride=true
    gravityScale=0.35
    simulationSpace=RootRelative
    alphaScaleBiasClamp {bias=1, scale=-1}
    boundBone {boneName=dyn_L_weight_A}
    componentAppliedLinearAccClamp=<5000 5000 5000>
    componentLinearAccScale=<1 1 1>
    componentLinearVelScale=<1 1 1>
    gravityOverride {z=-800}
    physicsBodyDefinitions [{
      boundBone {boneName=dyn_L_weight_A}
      boxExtents=<11 11 11>
      constraintSetup {angularLimitsMax=<5 5 5>, angularLimitsMin=<-5 -5 -5>, angularTarget=<1 0 0>}
      localJointOffset=<0 0.1 -0.5>
     }]
    simSpaceSettings {maxAngularAcceleration=30, maxAngularVelocity=6, simSpaceAngularAlpha=1}
   }
   n3 = Anim Dynamics - Bone: dyn_L_weight_B(ComponentPose=@n2, Alpha=1, ExternalForce="0,0,0") {
    angularDampingOverride=0.999
    angularSpringConstant=2000
    bAngularSpring=true
    bOverrideAngularDamping=true
    bUseGravityOverride=true
    gravityScale=0.35
    simulationSpace=RootRelative
    alphaScaleBiasClamp {bias=1, scale=-1}
    boundBone {boneName=dyn_L_weight_B}
    componentAppliedLinearAccClamp=<5000 5000 5000>
    componentLinearAccScale=<1 1 1>
    componentLinearVelScale=<1 1 1>
    gravityOverride {z=-800}
    physicsBodyDefinitions [{
      boundBone {boneName=dyn_L_weight_B}
      boxExtents=<13 13 13>
      constraintSetup {angularLimitsMax=<5 5 5>, angularLimitsMin=<-5 -5 -5>, angularTarget=<1 0 0>}
      localJointOffset=<0 0.1 -0.5>
     }]
    simSpaceSettings {maxAngularAcceleration=30, maxAngularVelocity=6, simSpaceAngularAlpha=1}
   }
   n4 = Anim Dynamics - Bone: dyn_L_weight_C(ComponentPose=@n3, Alpha=1, ExternalForce="0,0,0") {
    angularDampingOverride=0.999
    angularSpringConstant=2000
    bAngularSpring=true
    bOverrideAngularDamping=true
    bUseGravityOverride=true
    gravityScale=0.35
    simulationSpace=RootRelative
    alphaScaleBiasClamp {bias=1, scale=-1}
    boundBone {boneName=dyn_L_weight_C}
    componentAppliedLinearAccClamp=<5000 5000 5000>
    componentLinearAccScale=<1 1 1>
    componentLinearVelScale=<1 1 1>
    gravityOverride {z=-800}
    physicsBodyDefinitions [{
      boundBone {boneName=dyn_L_weight_C}
      boxExtents=<12 12 12>
      constraintSetup {angularLimitsMax=<5 5 5>, angularLimitsMin=<-5 -5 -5>, angularTarget=<1 0 0>}
      localJointOffset=<0 0.1 -0.5>
     }]
    simSpaceSettings {maxAngularAcceleration=30, maxAngularVelocity=6, simSpaceAngularAlpha=1}
   }
   n5 = Anim Dynamics - Bone: dyn_R_weight_A(ComponentPose=@n4, Alpha=1, ExternalForce="0,0,0") {
    angularDampingOverride=0.999
    angularSpringConstant=2000
    bAngularSpring=true
    bOverrideAngularDamping=true
    bUseGravityOverride=true
    gravityScale=0.35
    simulationSpace=RootRelative
    alphaScaleBiasClamp {bias=1, scale=-1}
    boundBone {boneName=dyn_R_weight_A}
    componentAppliedLinearAccClamp=<5000 5000 5000>
    componentLinearAccScale=<1 1 1>
    componentLinearVelScale=<1 1 1>
    gravityOverride {z=-800}
    physicsBodyDefinitions [{
      boundBone {boneName=dyn_R_weight_A}
      boxExtents=<12 12 12>
      constraintSetup {angularLimitsMax=<5 5 5>, angularLimitsMin=<-5 -5 -5>, angularTarget=<1 0 0>}
      localJointOffset=<0 -0.1 0.5>
     }]
    simSpaceSettings {maxAngularAcceleration=30, maxAngularVelocity=6, simSpaceAngularAlpha=1}
   }
   n6 = Anim Dynamics - Bone: dyn_R_weight_B(ComponentPose=@n5, Alpha=1, ExternalForce="0,0,0") {
    angularDampingOverride=0.999
    angularSpringConstant=2000
    bAngularSpring=true
    bOverrideAngularDamping=true
    bUseGravityOverride=true
    gravityScale=0.35
    simulationSpace=RootRelative
    alphaScaleBiasClamp {bias=1, scale=-1}
    boundBone {boneName=dyn_R_weight_B}
    componentAppliedLinearAccClamp=<5000 5000 5000>
    componentLinearAccScale=<1 1 1>
    componentLinearVelScale=<1 1 1>
    gravityOverride {z=-800}
    physicsBodyDefinitions [{
      boundBone {boneName=dyn_R_weight_B}
      boxExtents=<13 13 13>
      constraintSetup {angularLimitsMax=<5 5 5>, angularLimitsMin=<-5 -5 -5>, angularTarget=<1 0 0>}
      localJointOffset=<0 -0.1 0.5>
     }]
    simSpaceSettings {maxAngularAcceleration=30, maxAngularVelocity=6, simSpaceAngularAlpha=1}
   }
   n7 = Anim Dynamics - Bone: dyn_R_weight_C(ComponentPose=@n6, Alpha=1, ExternalForce="0,0,0") {
    angularDampingOverride=0.999
    angularSpringConstant=2000
    bAngularSpring=true
    bOverrideAngularDamping=true
    bUseGravityOverride=true
    gravityScale=0.35
    simulationSpace=RootRelative
    alphaScaleBiasClamp {bias=1, scale=-1}
    boundBone {boneName=dyn_R_weight_C}
    componentAppliedLinearAccClamp=<5000 5000 5000>
    componentLinearAccScale=<1 1 1>
    componentLinearVelScale=<1 1 1>
    gravityOverride {z=-800}
    physicsBodyDefinitions [{
      boundBone {boneName=dyn_R_weight_C}
      boxExtents=<11 11 11>
      constraintSetup {angularLimitsMax=<5 5 5>, angularLimitsMin=<-5 -5 -5>, angularTarget=<1 0 0>}
      localJointOffset=<0 -0.1 0.5>
     }]
    simSpaceSettings {maxAngularAcceleration=30, maxAngularVelocity=6, simSpaceAngularAlpha=1}
   }
   n8 = Anim Dynamics - Bone: dyn_C_front_AddPhysics!(ComponentPose=@n7, Alpha=1) -> (Pose) {
    angularDampingOverride=1
    angularSpringConstant=3000
    bAngularSpring=true
    bOverrideAngularDamping=true
    bUseGravityOverride=true
    gravityScale=0.35
    simulationSpace=RootRelative
    alphaScaleBiasClamp {bias=1, scale=-1}
    boundBone {boneName=dyn_C_front_AddPhysics}
    componentAppliedLinearAccClamp=<1000 1000 1000>
    componentLinearAccScale=<1 1 1>
    componentLinearVelScale=<1 1 1>
    gravityOverride {z=-800}
    physicsBodyDefinitions [{
      boundBone {boneName=dyn_C_front_AddPhysics}
      boxExtents=<11 11 11>
      constraintSetup {angularLimitsMax=<5 5 5>, angularLimitsMin=<-5 -5 -5>, angularTarget=<1 0 0>}
      localJointOffset=<1 0 0>
     }]
    simSpaceSettings {maxAngularAcceleration=15, maxAngularVelocity=3, simSpaceAngularAlpha=1}
   }
   n9 = Anim Dynamics - Bone: dyn_L_side_AddPhysics!(ComponentPose=@n8, Alpha=1) -> (Pose) {
    angularDampingOverride=1
    angularSpringConstant=3000
    bAngularSpring=true
    bOverrideAngularDamping=true
    bUseGravityOverride=true
    gravityScale=0.35
    simulationSpace=RootRelative
    alphaScaleBiasClamp {bias=1, scale=-1}
    boundBone {boneName=dyn_L_side_AddPhysics}
    componentAppliedLinearAccClamp=<1000 1000 1000>
    componentLinearAccScale=<1 1 1>
    componentLinearVelScale=<1 1 1>
    gravityOverride {z=-800}
    physicsBodyDefinitions [{
      boundBone {boneName=dyn_L_side_AddPhysics}
      boxExtents=<13 13 13>
      constraintSetup {angularLimitsMax=<5 5 5>, angularLimitsMin=<-5 -5 -5>, angularTarget=<1 0 0>}
      localJointOffset=<1 0 0>
     }]
    simSpaceSettings {maxAngularAcceleration=15, maxAngularVelocity=3, simSpaceAngularAlpha=1}
   }
   n10 = Anim Dynamics - Bone: dyn_R_side_AddPhysics!(ComponentPose=@n9, Alpha=1) -> (Pose) {
    angularDampingOverride=1
    angularSpringConstant=3000
    bAngularSpring=true
    bOverrideAngularDamping=true
    bUseGravityOverride=true
    gravityScale=0.35
    simulationSpace=RootRelative
    alphaScaleBiasClamp {bias=1, scale=-1}
    boundBone {boneName=dyn_R_side_AddPhysics}
    componentAppliedLinearAccClamp=<1000 1000 1000>
    componentLinearAccScale=<1 1 1>
    componentLinearVelScale=<1 1 1>
    gravityOverride {z=-800}
    physicsBodyDefinitions [{
      boundBone {boneName=dyn_R_side_AddPhysics}
      boxExtents=<11 11 11>
      constraintSetup {angularLimitsMax=<5 5 5>, angularLimitsMin=<-5 -5 -5>, angularTarget=<1 0 0>}
      localJointOffset=<-1 0 0>
     }]
    simSpaceSettings {maxAngularAcceleration=15, maxAngularVelocity=3, simSpaceAngularAlpha=1}
   }
   n11 = "Transform (Modify) Bone - Bone: dyn_R_side_AddPhysics"!(ComponentPose=@n10, Alpha=1) -> (Pose) {
    translationMode=BMM_Additive
    translationSpace=BCS_BoneSpace
    alphaScaleBiasClamp {bias=1, scale=-1}
    boneToModify {boneName=dyn_R_side_AddPhysics}
    translation {z=-3}
   }
   n12 = Anim Dynamics - Bone: dyn_C_front_AddPhysics!(ComponentPose=@n11, Alpha=1, GravityOverride:Vector="0,0,-200") -> (Pose) {
    angularBiasOverride=1
    angularDampingOverride=0.99
    angularSpringConstant=10000
    bGravityOverrideInSimSpace=true
    bLinearSpring=true
    bOverrideAngularBias=true
    bOverrideLinearDamping=true
    bUseGravityOverride=true
    linearDampingOverride=1
    linearSpringConstant=180
    simulationSpace=RootRelative
    alphaScaleBiasClamp {bias=1, scale=-1}
    boundBone {boneName=dyn_C_front_AddPhysics}
    componentAppliedLinearAccClamp=<5 5 5>
    gravityOverride {z=-200}
    physicsBodyDefinitions [{
      boundBone {boneName=dyn_C_front_AddPhysics}
      boxExtents=<80 80 80>
      constraintSetup {angularLimitsMax=<0 0 0>, angularLimitsMin=<0 0 0>, angularTarget=<1 0 0>, linearAxesMax=<0 0.5 5>, linearAxesMin=<0 -0.5 -0.5>}
      localJointOffset=<0.2 0 0>
     }]
    relativeSpaceBone {boneName=spine_04}
    simSpaceSettings {maxAngularAcceleration=50, maxAngularVelocity=10, simSpaceAngularAlpha=0.1}
   }
   n13 = Anim Dynamics - Bone: dyn_L_side_AddPhysics!(ComponentPose=@n12, Alpha=1, GravityOverride:Vector="0,0,-200") -> (Pose) {
    angularBiasOverride=1
    angularDampingOverride=0.99
    angularSpringConstant=10000
    bGravityOverrideInSimSpace=true
    bLinearSpring=true
    bOverrideAngularBias=true
    bOverrideLinearDamping=true
    bUseGravityOverride=true
    linearDampingOverride=1
    linearSpringConstant=300
    simulationSpace=RootRelative
    alphaScaleBiasClamp {bias=1, scale=-1}
    boundBone {boneName=dyn_L_side_AddPhysics}
    componentAppliedLinearAccClamp=<5 5 5>
    gravityOverride {z=-200}
    physicsBodyDefinitions [{
      boundBone {boneName=dyn_L_side_AddPhysics}
      boxExtents=<150 150 150>
      constraintSetup {angularLimitsMax=<0 0 0>, angularLimitsMin=<0 0 0>, angularTarget=<1 0 0>, linearAxesMax=<0 0.5 2>, linearAxesMin=<0 -0.5 -0.5>}
      localJointOffset=<0.2 0 0>
     }]
    relativeSpaceBone {boneName=spine_04}
    simSpaceSettings {maxAngularAcceleration=50, maxAngularVelocity=10, simSpaceAngularAlpha=0.1}
   }
   n14 = Anim Dynamics - Bone: dyn_R_side_AddPhysics!(ComponentPose=@n13, Alpha=1, GravityOverride:Vector="0,0,-200") -> (Pose) {
    angularBiasOverride=1
    angularDampingOverride=0.99
    angularSpringConstant=10000
    bGravityOverrideInSimSpace=true
    bLinearSpring=true
    bOverrideAngularBias=true
    bOverrideLinearDamping=true
    bUseGravityOverride=true
    linearDampingOverride=1
    linearSpringConstant=300
    simulationSpace=RootRelative
    alphaScaleBiasClamp {bias=1, scale=-1}
    boundBone {boneName=dyn_R_side_AddPhysics}
    componentAppliedLinearAccClamp=<5 5 5>
    gravityOverride {z=-200}
    physicsBodyDefinitions [{
      boundBone {boneName=dyn_R_side_AddPhysics}
      boxExtents=<150 150 150>
      constraintSetup {angularLimitsMax=<0 0 0>, angularLimitsMin=<0 0 0>, angularTarget=<1 0 0>, linearAxesMax=<0 0.5 0.5>, linearAxesMin=<0 -0.5 -2>}
      localJointOffset=<0.2 0 0>
     }]
    relativeSpaceBone {boneName=spine_04}
    simSpaceSettings {maxAngularAcceleration=50, maxAngularVelocity=10, simSpaceAngularAlpha=0.1}
   }
   n15 = "Constraint - dyn_L_weight_A (PPPPPPPP)"!(
    ComponentPose=@n14
    Alpha=1
    ConstraintWeights_0=0.2743
    ConstraintWeights_1:float=0.2647
    ConstraintWeights_2:float=0.2558
    ConstraintWeights_3:float=0.0994
    ConstraintWeights_4:float=0.0401
    ConstraintWeights_5:float=0.0312
    ConstraintWeights_6:float=0.0214
    ConstraintWeights_7:float=0.0132
   ) -> (Pose) {
    boneToModify {boneName=dyn_L_weight_A}
    constraintSetup [
     {targetBone {boneName=dyn_C_front_AddPhysics}}
     {targetBone {boneName=dyn_L_groinAdjust}}
     {targetBone {boneName=dyn_L_side_AddPhysics}}
     {targetBone {boneName=spine_01}}
     {targetBone {boneName=thigh_l}}
     {targetBone {boneName=spine_02}}
     {targetBone {boneName=spine_03}}
     {targetBone {boneName=pelvis}}
    ]
    constraintWeights [1,1,1,1,1,1,1,1]
   }
   n16 = "Constraint - dyn_L_weight_B (PPPPPPPP)"!(
    ComponentPose=@n15
    Alpha=1
    ConstraintWeights_0=0.5626
    ConstraintWeights_1:float=0.2188
    ConstraintWeights_2:float=0.0659
    ConstraintWeights_3:float=0.0413
    ConstraintWeights_4:float=0.0384
    ConstraintWeights_5:float=0.0364
    ConstraintWeights_6:float=0.0255
    ConstraintWeights_7:float=0.011
   ) -> (Pose) {
    boneToModify {boneName=dyn_L_weight_B}
    constraintSetup [
     {targetBone {boneName=dyn_L_side_AddPhysics}}
     {targetBone {boneName=spine_01}}
     {targetBone {boneName=dyn_C_front_AddPhysics}}
     {targetBone {boneName=pelvis}}
     {targetBone {boneName=thigh_l}}
     {targetBone {boneName=spine_02}}
     {targetBone {boneName=spine_03}}
     {targetBone {boneName=dyn_R_side_AddPhysics}}
    ]
    constraintWeights [1,1,1,1,1,1,1,1]
   }
   n17 = "Constraint - dyn_L_weight_C (PPPPPPPP)"!(
    ComponentPose=@n16
    Alpha=1
    ConstraintWeights_0=0.4852
    ConstraintWeights_1:float=0.1606
    ConstraintWeights_2:float=0.1081
    ConstraintWeights_3:float=0.0656
    ConstraintWeights_4:float=0.0613
    ConstraintWeights_5:float=0.0594
    ConstraintWeights_6:float=0.0376
    ConstraintWeights_7:float=0.0223
   ) -> (Pose) {
    boneToModify {boneName=dyn_L_weight_C}
    constraintSetup [
     {targetBone {boneName=spine_01}}
     {targetBone {boneName=pelvis}}
     {targetBone {boneName=dyn_L_side_AddPhysics}}
     {targetBone {boneName=deform_glute_l}}
     {targetBone {boneName=spine_02}}
     {targetBone {boneName=dyn_R_side_AddPhysics}}
     {targetBone {boneName=spine_03}}
     {targetBone {boneName=spine_04}}
    ]
    constraintWeights [1,1,1,1,1,1,1,1]
   }
   n18 = "Constraint - dyn_R_weight_A (PPPPPPPP)"!(
    ComponentPose=@n17
    Alpha=1
    ConstraintWeights_0=0.2743
    ConstraintWeights_1:float=0.2647
    ConstraintWeights_2:float=0.2558
    ConstraintWeights_3:float=0.0994
    ConstraintWeights_4:float=0.0401
    ConstraintWeights_5:float=0.0312
    ConstraintWeights_6:float=0.0214
    ConstraintWeights_7:float=0.0132
   ) -> (Pose) {
    boneToModify {boneName=dyn_R_weight_A}
    constraintSetup [
     {targetBone {boneName=dyn_C_front_AddPhysics}}
     {targetBone {boneName=dyn_R_groinAdjust}}
     {targetBone {boneName=dyn_R_side_AddPhysics}}
     {targetBone {boneName=spine_01}}
     {targetBone {boneName=thigh_r}}
     {targetBone {boneName=spine_02}}
     {targetBone {boneName=spine_03}}
     {targetBone {boneName=pelvis}}
    ]
    constraintWeights [1,1,1,1,1,1,1,1]
   }
   n19 = "Constraint - dyn_R_weight_B (PPPPPPPP)"!(
    ComponentPose=@n18
    Alpha=1
    ConstraintWeights_0=0.5626
    ConstraintWeights_1:float=0.2188
    ConstraintWeights_2:float=0.0659
    ConstraintWeights_3:float=0.0413
    ConstraintWeights_4:float=0.0384
    ConstraintWeights_5:float=0.0364
    ConstraintWeights_6:float=0.0255
    ConstraintWeights_7:float=0.011
   ) -> (Pose) {
    boneToModify {boneName=dyn_R_weight_B}
    constraintSetup [
     {targetBone {boneName=dyn_R_side_AddPhysics}}
     {targetBone {boneName=spine_01}}
     {targetBone {boneName=dyn_C_front_AddPhysics}}
     {targetBone {boneName=pelvis}}
     {targetBone {boneName=thigh_r}}
     {targetBone {boneName=spine_02}}
     {targetBone {boneName=spine_03}}
     {targetBone {boneName=dyn_L_side_AddPhysics}}
    ]
    constraintWeights [1,1,1,1,1,1,1,1]
   }
   n20 = "Constraint - dyn_R_weight_C (PPPPPPPP)"!(
    ComponentPose=@n19
    Alpha=1
    ConstraintWeights_0=0.4852
    ConstraintWeights_1:float=0.1606
    ConstraintWeights_2:float=0.1081
    ConstraintWeights_3:float=0.0656
    ConstraintWeights_4:float=0.0613
    ConstraintWeights_5:float=0.0594
    ConstraintWeights_6:float=0.0376
    ConstraintWeights_7:float=0.0223
   ) -> (Pose) {
    boneToModify {boneName=dyn_R_weight_C}
    constraintSetup [
     {targetBone {boneName=spine_01}}
     {targetBone {boneName=pelvis}}
     {targetBone {boneName=dyn_R_side_AddPhysics}}
     {targetBone {boneName=deform_glute_r}}
     {targetBone {boneName=spine_02}}
     {targetBone {boneName=dyn_L_side_AddPhysics}}
     {targetBone {boneName=spine_03}}
     {targetBone {boneName=spine_04}}
    ]
    constraintWeights [1,1,1,1,1,1,1,1]
   }
   Output Pose(Result=Component To Local(ComponentPose=@n20)) {layerGroup=None}
  ]}
 EventGraph {nodes [
   Event OnEndSliding() {is=title K2Node_Event}
   > Set slidingAlpha(slidingAlpha:double=0) -> (Output_Get:double)
   Event OnBeginSliding() {is=title K2Node_Event}
   > Set slidingAlpha(slidingAlpha=1)
  ]}
]
variables [slidingAlpha {type=number}]
