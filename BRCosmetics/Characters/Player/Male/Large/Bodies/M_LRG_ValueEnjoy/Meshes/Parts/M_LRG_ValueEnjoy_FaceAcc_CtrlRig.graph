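# Control Rig for the FaceAcc part of the large male player body M_LRG_ValueEnjoy, adding corrective motion to the clavicle deform bones.
# Each side reads the clavicle rotation relative to spine_05, converts it to Euler, and remaps the Y angle to a weight.
# That weight blends a mirrored translation offset onto deform_clavicle_l or deform_clavicle_r.
# Facial and upper-arm bones are in the hierarchy, but the graph does not drive them.
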ControlRigBlueprint /BRCosmetics/Characters/Player/Male/Large/Bodies/M_LRG_ValueEnjoy/Meshes/Parts/M_LRG_ValueEnjoy_FaceAcc_CtrlRig
graphs [
 RigVMFunctionLibrary {links [], nodes []}
 RigVMModel {nodes [
   RigUnit_BeginExecution = &RigUnit_BeginExecution()
   n1 = &RigUnit_GetRelativeTransformForItem(Child=(Type=Bone,Name="clavicle_l"), Parent=(Type=Bone,Name="spine_05"), bChildInitial=False, bParentInitial=False)
   &RigUnit_ModifyTransforms(
    ItemToModify=((Item=(Type=Bone,Name="deform_clavicle_l"),Transform=(Rotation=(X=0,Y=0,Z=0,W=-1),Translation=(X=9,Y=0,Z=-2),Scale3D=(X=1,Y=1,Z=1))))
    WeightMinimum=0
    ExecutePin=@RigUnit_BeginExecution.ExecutePin
    Weight=&RigVMFunction_MathDoubleRemap(
     SourceMaximum=55
     SourceMinimum=80
     TargetMaximum=1
     TargetMinimum=0
     Value=&RigVMFunction_MathQuaternionToEuler(RotationOrder=ZXY, Value=@n1.RelativeTransform.Rotation).Result.Y
    ).Result
   )
   n2 = &RigUnit_GetRelativeTransformForItem(Child=(Type=Bone,Name="clavicle_r"), Parent=(Type=Bone,Name="spine_05"), bChildInitial=False, bParentInitial=False)
   > &RigUnit_ModifyTransforms(
    ItemToModify=((Item=(Type=Bone,Name="deform_clavicle_r"),Transform=(Rotation=(X=0,Y=0,Z=0,W=-1),Translation=(X=-9,Y=0,Z=2),Scale3D=(X=1,Y=1,Z=1))))
    WeightMinimum=0
    Weight=&RigVMFunction_MathDoubleRemap(
     SourceMaximum=-55
     SourceMinimum=-80
     TargetMaximum=1
     TargetMinimum=0
     Value=&RigVMFunction_MathQuaternionToEuler(RotationOrder=ZXY, Value=@n2.RelativeTransform.Rotation).Result.Y
    ).Result
   )
  ]}
]
hierarchy [
 root {loc=<0 0 0>, rot=<0 0 0 1>}
 pelvis {parent=root, loc=<0 -0.7321 92.4218>, rot=<0.0127 0.707 -0.0127 -0.707>}
 spine_01 {parent=pelvis, loc=<5.6002 -0.5128 0>, rot=<0 0 0.034 -0.9994>}
 spine_02 {parent=spine_01, loc=<8.0756 0.0139 0>, rot=<0 0 0.0238 -0.9997>}
 spine_03 {parent=spine_02, loc=<7.0395 0 0>, rot=<0 0 0.0149 -0.9999>}
 spine_04 {parent=spine_03, loc=<8.1023 0 0>, rot=<0 0 -0.0132 -0.9999>}
 spine_05 {parent=spine_04, loc=<10.0141 0 0>, rot=<0 0 -0.0015 -1>}
 clavicle_l {parent=spine_05, loc=<2.5856 -0.2866 -4.3777>, rot=<0.0465 -0.7289 -0.009 -0.6829>}
 deform_upperarm_l {parent=clavicle_l, loc=<18.2264 0.1342 -0.001>, rot=<0.0253 -0.2693 0.0013 -0.9627>}
 deform_upperarm_twist_01_l {parent=deform_upperarm_l, loc=<7.9796 0.0843 -0.6195>, rot=<0 0 0 -1>}
 deform_upperarm_bk_base_l {parent=deform_upperarm_twist_01_l, loc=<-2.8248 -8.2399 2.7028>, rot=<-0.0075 0.0004 -0.0178 -0.9998>}
 deform_upperarm_bk_l {parent=deform_upperarm_bk_base_l, loc=<0 0 0>, rot=<0 0 0 1>}
 deform_upperarm_up_base_l {parent=deform_upperarm_twist_01_l, loc=<-6.9436 0 6.0472>, rot=<-0.0076 0.0057 -0.0178 -0.9998>}
 deform_upperarm_up_l {parent=deform_upperarm_up_base_l, loc=<0 0 0>, rot=<0 0 0 1>}
 deform_clavicle_l {parent=clavicle_l, loc=<7 0 9.5>, rot=<0 0 0 1>}
 neck_01 {parent=spine_05, loc=<10.815 1.9196 0>, rot=<0 0 -0.1864 -0.9825>}
 neck_02 {parent=neck_01, loc=<4.7347 0.3045 0>, rot=<0 0 0.0363 -0.9993>}
 head {parent=neck_02, loc=<5.7632 0.5552 0>, rot=<0 0 0.115 -0.9934>}
 FACIAL_C_FacialRoot {parent=head, loc=<2.1743 -3.3889 -0.0002>, rot=<0.4879 -0.5118 -0.4879 -0.5118>}
 FACIAL_C_Forehead {parent=FACIAL_C_FacialRoot, loc=<-0.0077 -2.6623 12.0306>, rot=<-0.0094 -0.0002 0.0008 -1>}
 FACIAL_L_ForeheadIn {parent=FACIAL_C_FacialRoot, loc=<1.8284 -2.4837 12.0253>, rot=<-0.0269 -0.0696 0.0075 -0.9972>}
 FACIAL_R_ForeheadIn {parent=FACIAL_C_FacialRoot, loc=<-1.8386 -2.4963 12.0211>, rot=<-0.0268 0.0692 -0.0058 -0.9972>}
 FACIAL_L_ForeheadMid {parent=FACIAL_C_FacialRoot, loc=<3.2953 -2.667 11.6212>, rot=<-0.0377 -0.1764 -0.0138 -0.9835>}
 FACIAL_R_ForeheadMid {parent=FACIAL_C_FacialRoot, loc=<-3.3029 -2.669 11.6111>, rot=<-0.0374 0.1761 0.0155 -0.9835>}
 FACIAL_L_ForeheadOut {parent=FACIAL_C_FacialRoot, loc=<4.7479 -2.4864 10.9936>, rot=<-0.089 -0.2583 -0.0696 -0.9594>}
 FACIAL_R_ForeheadOut {parent=FACIAL_C_FacialRoot, loc=<-4.7636 -2.4754 10.9782>, rot=<-0.0886 0.258 0.0712 -0.9594>}
 FACIAL_L_EyesackUpper {parent=FACIAL_C_FacialRoot, loc=<3.3368 -1.686 11.8967>, rot=<0.0825 -0.0867 -0.0608 -0.991>}
 FACIAL_R_EyesackUpper {parent=FACIAL_C_FacialRoot, loc=<-3.3254 -1.7024 11.8914>, rot=<0.0826 0.0861 0.0624 -0.9909>}
 FACIAL_L_EyelidUpperFurrow {parent=FACIAL_C_FacialRoot, loc=<3.2768 -1.2745 11.7633>, rot=<0.0033 -0.1537 -0.0737 -0.9854>}
 FACIAL_R_EyelidUpperFurrow {parent=FACIAL_C_FacialRoot, loc=<-3.2713 -1.2888 11.7793>, rot=<0.0036 0.1532 0.0753 -0.9853>}
 FACIAL_L_EyeCornerOuter {parent=FACIAL_C_FacialRoot, loc=<4.8615 -0.3213 10.8644>, rot=<0.0311 -0.3826 0.0138 -0.9233>}
 FACIAL_R_EyeCornerOuter {parent=FACIAL_C_FacialRoot, loc=<-4.8914 -0.3041 10.846>, rot=<0.0317 0.3821 -0.0122 -0.9235>}
 FACIAL_L_CheekInner {parent=FACIAL_C_FacialRoot, loc=<2.6656 1.4995 11.4478>, rot=<-0.0903 -0.0948 -0.1932 -0.9724>}
 FACIAL_R_CheekInner {parent=FACIAL_C_FacialRoot, loc=<-2.6641 1.4715 11.4333>, rot=<-0.0901 0.0945 0.1948 -0.9721>}
 FACIAL_L_CheekOuter {parent=FACIAL_C_FacialRoot, loc=<4.9637 1.7142 9.8953>, rot=<0.0861 -0.3736 0.1139 -0.9165>}
 FACIAL_R_CheekOuter {parent=FACIAL_C_FacialRoot, loc=<-4.9546 1.7121 9.8827>, rot=<0.0867 0.373 -0.1124 -0.9169>}
 FACIAL_L_NasolabialBulge {parent=FACIAL_C_FacialRoot, loc=<4.456 5.0008 10.5679>, rot=<0.089 -0.3766 0.0374 -0.9213>}
 FACIAL_R_NasolabialBulge {parent=FACIAL_C_FacialRoot, loc=<-4.4383 5.0081 10.5769>, rot=<0.0896 0.376 -0.0359 -0.9216>}
 FACIAL_L_NasolabialFurrow {parent=FACIAL_C_FacialRoot, loc=<3.3793 3.4444 12.9522>, rot=<-0.0021 -0.1383 -0.3584 -0.9233>}
 FACIAL_R_NasolabialFurrow {parent=FACIAL_C_FacialRoot, loc=<-3.3742 3.4169 12.9408>, rot=<-0.0017 0.1379 0.3599 -0.9227>}
 FACIAL_L_CheekLower {parent=FACIAL_C_FacialRoot, loc=<4.7927 4.9504 9.7912>, rot=<0.0649 -0.4249 0.073 -0.9>}
 FACIAL_R_CheekLower {parent=FACIAL_C_FacialRoot, loc=<-4.7765 4.9608 9.8051>, rot=<0.0656 0.4244 -0.0715 -0.9003>}
 FACIAL_C_Nose {parent=FACIAL_C_FacialRoot, loc=<-0.0021 0.6193 13.0063>, rot=<-0.0966 -0.0001 0.0009 -0.9953>}
 FACIAL_C_NoseLower {parent=FACIAL_C_Nose, loc=<-0.0091 2.9635 0.4376>, rot=<0.1305 0 0 -0.9914>}
 FACIAL_L_Nostril {parent=FACIAL_C_Nose, loc=<1.0797 2.2154 0.5124>, rot=<0.0173 0.1305 0.0023 -0.9913>}
 FACIAL_R_Nostril {parent=FACIAL_C_Nose, loc=<-1.0833 2.2024 0.5226>, rot=<0.0173 -0.1305 -0.0023 -0.9913>}
 FACIAL_C_MouthUpper {parent=FACIAL_C_FacialRoot, loc=<0.0084 5.2947 11.5525>, rot=<0.034 -0.0002 0.0008 -0.9994>}
 FACIAL_C_LipUpper {parent=FACIAL_C_MouthUpper, loc=<0.0074 -0.8763 2.141>, rot=<-0.0872 0 0 -0.9962>}
 FACIAL_L_LipUpper {parent=FACIAL_C_MouthUpper, loc=<1.2478 -0.7941 1.8879>, rot=<-0.0778 -0.1601 -0.0379 -0.9833>}
 FACIAL_R_LipUpper {parent=FACIAL_C_MouthUpper, loc=<-1.2381 -0.8105 1.8984>, rot=<-0.0778 0.1601 0.0379 -0.9833>}
 FACIAL_L_LipUpperOuter {parent=FACIAL_C_MouthUpper, loc=<2.2016 -0.4264 1.3674>, rot=<-0.0984 -0.2293 -0.0983 -0.9634>}
 FACIAL_R_LipUpperOuter {parent=FACIAL_C_MouthUpper, loc=<-2.1932 -0.4459 1.3739>, rot=<-0.0984 0.2293 0.0983 -0.9634>}
 FACIAL_L_LipCorner {parent=FACIAL_C_MouthUpper, loc=<2.7173 0.0213 0.8123>, rot=<0 -0.2078 0 -0.9782>}
 FACIAL_R_LipCorner {parent=FACIAL_C_MouthUpper, loc=<-2.7147 -0.0047 0.7866>, rot=<0 0.2078 0 -0.9782>}
 FACIAL_L_Masseter {parent=FACIAL_C_FacialRoot, loc=<8.2599 2.4369 7.2716>, rot=<0.1031 0.1128 0.0048 -0.9882>}
 FACIAL_R_Masseter {parent=FACIAL_C_FacialRoot, loc=<-8.246 2.4783 7.269>, rot=<0.1029 -0.1134 -0.0032 -0.9882>}
 FACIAL_C_UnderChin {parent=FACIAL_C_FacialRoot, loc=<0.0047 13.1736 10.0703>, rot=<0.529 -0.0006 0.0006 -0.8486>}
 FACIAL_L_UnderChin {parent=FACIAL_C_FacialRoot, loc=<5.6846 10.683 7.1451>, rot=<0.4306 -0.3524 0.0949 -0.8255>}
 FACIAL_R_UnderChin {parent=FACIAL_C_FacialRoot, loc=<-5.6802 10.7028 7.154>, rot=<0.4311 0.3513 -0.0937 -0.8258>}
 FACIAL_C_LowerLipRotation {parent=FACIAL_C_FacialRoot, loc=<0.0097 2.8469 5.4696>, rot=<0.3738 -0.0005 0.0007 -0.9275>}
 FACIAL_C_MouthLower {parent=FACIAL_C_LowerLipRotation, loc=<-0.0081 -2.4541 6.0804>, rot=<-0.342 0 0 -0.9397>}
 FACIAL_C_LipLower {parent=FACIAL_C_MouthLower, loc=<0.0023 0.6729 2.1153>, rot=<0.0872 0 0 -0.9962>}
 FACIAL_L_LipLower {parent=FACIAL_C_MouthLower, loc=<1.1731 0.6745 1.9457>, rot=<0.0927 -0.1294 0.0272 -0.9869>}
 FACIAL_R_LipLower {parent=FACIAL_C_MouthLower, loc=<-1.1677 0.6584 1.9475>, rot=<0.0927 0.1294 -0.0272 -0.9869>}
 FACIAL_L_LipLowerOuter {parent=FACIAL_C_MouthLower, loc=<2.1165 0.4642 1.3987>, rot=<0.0913 -0.2321 0.0318 -0.9679>}
 FACIAL_R_LipLowerOuter {parent=FACIAL_C_MouthLower, loc=<-2.1226 0.4604 1.417>, rot=<0.0913 0.2321 -0.0318 -0.9679>}
 FACIAL_C_Jaw {parent=FACIAL_C_FacialRoot, loc=<0.0097 2.8469 5.4696>, rot=<0.3738 -0.0005 0.0007 -0.9275>}
 FACIAL_C_Jawline {parent=FACIAL_C_Jaw, loc=<-0.0187 1.9341 11.2403>, rot=<0.2586 0 0 -0.966>}
 FACIAL_L_ChinSide {parent=FACIAL_C_Jaw, loc=<3.8608 -0.3456 8.9088>, rot=<-0.2208 -0.3738 0.0499 -0.8995>}
 FACIAL_R_ChinSide {parent=FACIAL_C_Jaw, loc=<-3.8725 -0.3452 8.9168>, rot=<-0.2208 0.3738 -0.0499 -0.8995>}
 FACIAL_C_Chin {parent=FACIAL_C_Jaw, loc=<-0.0132 0.1894 9.8929>, rot=<-0.2591 0 0 -0.9659>}
 FACIAL_L_Ear {parent=FACIAL_C_FacialRoot, loc=<7.7638 0.4161 3.9775>, rot=<0.0077 -0.0018 -0.0253 -0.9996>}
 FACIAL_R_Ear {parent=FACIAL_C_FacialRoot, loc=<-7.6702 0.4634 3.9408>, rot=<0.0077 0.0013 0.027 -0.9996>}
 FACIAL_L_Jawline {parent=FACIAL_C_FacialRoot, loc=<6.8124 7.7419 8.2627>, rot=<0.2421 0.1554 -0.0592 -0.9559>}
 FACIAL_R_Jawline {parent=FACIAL_C_FacialRoot, loc=<-6.8073 7.7641 8.2756>, rot=<0.2418 -0.1562 0.0607 -0.9557>}
 FACIAL_C_NoseBridge {parent=FACIAL_C_FacialRoot, loc=<-0.0142 -1.181 12.8754>, rot=<0.0778 -0.0003 0.0008 -0.997>}
 clavicle_r {parent=spine_05, loc=<2.5855 -0.2866 4.3777>, rot=<0.7289 0.0465 -0.6829 0.009>}
 deform_upperarm_r {parent=clavicle_r, loc=<-18.2265 -0.1342 0.0015>, rot=<0.0253 -0.2693 0.0013 -0.9627>}
 deform_upperarm_twist_01_r {parent=deform_upperarm_r, loc=<-7.9792 -0.0842 0.6189>, rot=<0 0 0 -1>}
 deform_upperarm_bk_base_r {parent=deform_upperarm_twist_01_r, loc=<2.8249 8.2398 -2.7028>, rot=<-0.0075 0.0004 -0.0178 -0.9998>}
 deform_upperarm_bk_r {parent=deform_upperarm_bk_base_r, loc=<0 0 0>, rot=<0 0 0 -1>}
 deform_upperarm_up_base_r {parent=deform_upperarm_twist_01_r, loc=<6.9433 0 -6.0468>, rot=<-0.0076 0.0057 -0.0178 -0.9998>}
 deform_upperarm_up_r {parent=deform_upperarm_up_base_r, loc=<0 0 0>, rot=<0 0 0 1>}
 deform_clavicle_r {parent=clavicle_r, loc=<-7 0 -9.5>, rot=<0 0 0 1>}
]
