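# Animation Blueprint for the face accessory mesh part of the large male player cosmetic body ValueEnjoy.
# The AnimGraph copies the pose from the attached parent mesh, then runs it through the accessory's own Control Rig before output.
# On initialization it stores the owning skeletal mesh component in SkelMesh, which neither graph reads.
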AnimBlueprint /BRCosmetics/Characters/Player/Male/Large/Bodies/M_LRG_ValueEnjoy/Meshes/Parts/M_LRG_ValueEnjoy_FaceAcc_AnimBP
target_skeleton=&/M_LRG_ValueEnjoy_FaceAcc_Skeleton
graphs [
 AnimGraph {nodes [
   n1 = Control Rig(Source=Copy Pose From Mesh() {bUseAttachedParent=true}) {controlRigAssetReference {blueprintRigClass=&/M_LRG_ValueEnjoy_FaceAcc_CtrlRig.M_LRG_ValueEnjoy_FaceAcc_CtrlRig_C}}
   Output Pose(Result=@n1) {layerGroup=None}
  ]}
 EventGraph {nodes [
   Event BlueprintInitializeAnimation()
   > Set SkelMesh(SkelMesh=GetOwnerSkeletalMeshComponent())
  ]}
]
variables [SkelMesh {type=SkeletalMeshComponent}]
